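# Sound cue for the inflate sound when a player goes DBNO (down but not out) while swimming.
# A mixer layers the dedicated inflate wave, at normal level, with a slightly louder water layer.
# That layer randomly picks one of two reused third-person sprint-swim splashes, pitched up by a modulator for variation.
# It plays in the player foley sound class with the player swimming attenuation.
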SoundCue /Game/Athena/Sounds/Water/Interaction/Swim_DBNO_Inflate_Cue
root=Mixer_0
nodes [
 {id=Mixer_0, children [Modulator_0,WavePlayer_0], properties {inputVolume [1.2,1]}}
 {id=Modulator_0, children [Random_1], properties {pitchMax=1.4, pitchMin=1.2}}
 {id=Random_1, children [WavePlayer_1,WavePlayer_4], properties {weights [1,1]}}
 {id=WavePlayer_0, properties {soundWaveAssetPtr=&/Swim_DBNO_Inflate_01}}
 {id=WavePlayer_1, properties {soundWaveAssetPtr=&/Water_Player_Swim_Sprint_3P_10}}
 {id=WavePlayer_4, properties {soundWaveAssetPtr=&/Water_Player_Swim_Sprint_3P_09}}
]
properties {
 attenuationSettings=/Game/Sounds/Attenuation/Fort_Player_Swimming_Athena_Att
 maxDistance=4500
 soundClassObject=&Fort_Player_Foley
 volumeMultiplier=0.7467
}
