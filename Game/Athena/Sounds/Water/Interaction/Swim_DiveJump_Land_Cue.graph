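# Sound cue for the landing of a swimming dive jump into water, used as player foley in Athena.
# A random node picks one of three equally weighted landing wave variations on each play.
# It routes through the Athena player foley sound class and borrows the third-person sprint footstep attenuation to set its spatial falloff.
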SoundCue /Game/Athena/Sounds/Water/Interaction/Swim_DiveJump_Land_Cue
root=Random_0
nodes [
 {id=Random_0, children [WavePlayer_1,WavePlayer_2,WavePlayer_3], properties {weights [1,1,1]}}
 {id=WavePlayer_1, properties {soundWaveAssetPtr=&/Swim_DiveJump_Land_01}}
 {id=WavePlayer_2, properties {soundWaveAssetPtr=&/Swim_DiveJump_Land_02}}
 {id=WavePlayer_3, properties {soundWaveAssetPtr=&/Swim_DiveJump_Land_03}}
]
properties {
 attenuationSettings=/Game/Sounds/Attenuation/Fort_Player_Footsteps_3P_Sprint_Athena_Att
 maxDistance=3584
 soundClassObject=&Fort_Player_Foley_Athena
 volumeMultiplier=1
}
