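# Sound cue for the third-person swim jump sound of other players in Athena water, routed to the enemy player foley sound class.
# A Random node picks one of four equally weighted first-person swim jump waves, reused here, and feeds a Modulator that varies pitch.
# The cue lowers volume, raises base pitch, uses third-person jump attenuation and joins three swim concurrency groups to limit overlapping sounds.
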SoundCue /Game/Athena/Sounds/Water/Interaction/Water_Player_Swim_3P_Jump_Cue
root=Modulator_0
nodes [
 {id=Modulator_0, children [Random_0], properties {pitchMax=1.4, pitchMin=0.9}}
 {id=Random_0, children [WavePlayer_0,WavePlayer_1,WavePlayer_5,WavePlayer_6], properties {weights [1,1,1,1]}}
 {id=WavePlayer_0, properties {soundWaveAssetPtr=&/Water_Player_SwimJump_1P_01}}
 {id=WavePlayer_1, properties {soundWaveAssetPtr=&/Water_Player_SwimJump_1P_02}}
 {id=WavePlayer_5, properties {soundWaveAssetPtr=&/Water_Player_SwimJump_1P_03}}
 {id=WavePlayer_6, properties {soundWaveAssetPtr=&/Water_Player_SwimJump_1P_04}}
]
properties {
 attenuationSettings=/Game/Sounds/Attenuation/Fort_Player_Jump_3P_Athena_Att
 maxDistance=3584
 pitchMultiplier=1.2
 soundClassObject=&Fort_Player_Foley_3P_Enemy_Athena
 volumeMultiplier=0.4
 concurrencySet [
  &/Water_Player_Swim_3P_Concurrency_Group
  &/Water_Player_Swim_Land
  &/Conc_Water_Player_Swim_3P_All
 ]
}
