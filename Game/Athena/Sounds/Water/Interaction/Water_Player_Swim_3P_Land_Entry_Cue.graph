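# Sound cue for third-person splashes when a player enters water to swim from land.
# Branches on IsLocal, then IsEnemy, select local, enemy or quieter non-enemy mixes; unset IsEnemy falls back to enemy.
# Each mix crossfades randomized, modulated Slow, Medium and Fast wave sets by EntrySpeed, with EntrySpeed also scaling the local slow layer's volume.
# Enemy and non-enemy paths share water jump attenuation, and concurrency groups limit playback.
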SoundCue /Game/Athena/Sounds/Water/Interaction/Water_Player_Swim_3P_Land_Entry_Cue
root=Branch_0
nodes [
 {id=Attenuation_0, children [ParamCrossFade_0]}
 {
  id=Attenuation_1
  children [ParamCrossFade_2]
  properties {attenuationSettings=/Game/Sounds/Attenuation/Fort_Player_Jump_3P_Athena_Water_Stereo_Att}
 }
 {
  id=Attenuation_2
  children [ParamCrossFade_1]
  properties {attenuationSettings=&Fort_Player_Jump_3P_Athena_Water_Stereo_Att}
 }
 {id=Branch_0, children [Attenuation_0, Branch_1, Branch_1], properties {boolParameterName=IsLocal}}
 {id=Branch_1, children [Attenuation_1, Attenuation_2, Attenuation_1], properties {boolParameterName=IsEnemy}}
 {id=ModulatorContinuous_0, children [Modulator_2], properties {volumeModulationParams {default=0, maxInput=0.4, parameterName=EntrySpeed}}}
 {id=Modulator_0, children [Random_0], properties {volumeMax=0.85, volumeMin=0.75}}
 {id=Modulator_1, children [Random_1], properties {volumeMax=0.95, volumeMin=0.85}}
 {id=Modulator_2, children [Random_4], properties {volumeMin=0.9}}
 {id=Modulator_3, children [Random_5]}
 {id=Modulator_4, children [Random_6], properties {volumeMax=1.3}}
 {id=Modulator_5, children [Random_2], properties {volumeMax=1.1}}
 {id=Modulator_6, children [Random_3], properties {volumeMin=0.9}}
 {id=Modulator_7, children [Random_7]}
 {id=Modulator_8, children [Random_8], properties {volumeMax=1.1}}
 {
  id=ParamCrossFade_0
  children [ModulatorContinuous_0, Modulator_3, Modulator_4]
  properties {
   paramName=EntrySpeed
   crossFadeInput [
    {fadeInDistanceEnd=0, fadeInDistanceStart=0, fadeOutDistanceEnd=0.5, fadeOutDistanceStart=0.4, volume=1}
    {fadeInDistanceEnd=0.5, fadeInDistanceStart=0.4, fadeOutDistanceEnd=0.95, fadeOutDistanceStart=0.9, volume=1}
    {fadeInDistanceEnd=0.95, fadeInDistanceStart=0.9, fadeOutDistanceEnd=2, fadeOutDistanceStart=1, volume=1}
   ]
  }
 }
 {
  id=ParamCrossFade_1
  children [Modulator_0,Modulator_1,Modulator_5]
  properties {
   paramName=EntrySpeed
   crossFadeInput [
    {fadeInDistanceEnd=0, fadeInDistanceStart=0, fadeOutDistanceEnd=0.5, fadeOutDistanceStart=0.4, volume=0.25}
    {fadeInDistanceEnd=0.5, fadeInDistanceStart=0.4, fadeOutDistanceEnd=0.95, fadeOutDistanceStart=0.9, volume=0.25}
    {fadeInDistanceEnd=0.95, fadeInDistanceStart=0.9, fadeOutDistanceEnd=2, fadeOutDistanceStart=1, volume=0.35}
   ]
  }
 }
 {
  id=ParamCrossFade_2
  children [Modulator_6,Modulator_7,Modulator_8]
  properties {
   paramName=EntrySpeed
   crossFadeInput [
    {fadeInDistanceEnd=0, fadeInDistanceStart=0, fadeOutDistanceEnd=0.5, fadeOutDistanceStart=0.4, volume=0.6}
    {fadeInDistanceEnd=0.5, fadeInDistanceStart=0.4, fadeOutDistanceEnd=0.95, fadeOutDistanceStart=0.9, volume=0.6}
    {fadeInDistanceEnd=0.95, fadeInDistanceStart=0.9, fadeOutDistanceEnd=2, fadeOutDistanceStart=1, volume=0.8}
   ]
  }
 }
 {
  id=Random_0
  children [WavePlayer_0,WavePlayer_1,WavePlayer_2,WavePlayer_3]
  properties {bShouldExcludeFromBranchCulling=true, weights [1,1,1,1]}
 }
 {
  id=Random_1
  children [WavePlayer_4,WavePlayer_5,WavePlayer_6,WavePlayer_7]
  properties {bShouldExcludeFromBranchCulling=true, weights [1,1,1,1]}
 }
 {
  id=Random_2
  children [WavePlayer_8, WavePlayer_9, WavePlayer_10, WavePlayer_11]
  properties {bShouldExcludeFromBranchCulling=true, weights [1,1,1,1]}
 }
 {
  id=Random_3
  children [WavePlayer_12, WavePlayer_13, WavePlayer_14, WavePlayer_15]
  properties {bShouldExcludeFromBranchCulling=true, weights [1,1,1,1]}
 }
 {
  id=Random_4
  children [WavePlayer_16, WavePlayer_17, WavePlayer_18, WavePlayer_19]
  properties {bShouldExcludeFromBranchCulling=true, weights [1,1,1,1]}
 }
 {
  id=Random_5
  children [WavePlayer_20, WavePlayer_21, WavePlayer_22, WavePlayer_23]
  properties {bShouldExcludeFromBranchCulling=true, weights [1,1,1,1]}
 }
 {
  id=Random_6
  children [WavePlayer_24, WavePlayer_25, WavePlayer_26, WavePlayer_27]
  properties {bShouldExcludeFromBranchCulling=true, weights [1,1,1,1]}
 }
 {
  id=Random_7
  children [WavePlayer_28, WavePlayer_29, WavePlayer_30, WavePlayer_31]
  properties {bShouldExcludeFromBranchCulling=true, weights [1,1,1,1]}
 }
 {
  id=Random_8
  children [WavePlayer_32, WavePlayer_33, WavePlayer_34, WavePlayer_35]
  properties {bShouldExcludeFromBranchCulling=true, weights [1,1,1,1]}
 }
 {id=WavePlayer_0, properties {soundWaveAssetPtr=&/Water_Player_Swim_Entry_3P_Slow_01}}
 {id=WavePlayer_1, properties {soundWaveAssetPtr=&/Water_Player_Swim_Entry_3P_Slow_02}}
 {id=WavePlayer_10, properties {soundWaveAssetPtr=&/Water_Player_Swim_Entry_3P_Fast_03}}
 {id=WavePlayer_11, properties {soundWaveAssetPtr=&/Water_Player_Swim_Entry_3P_Fast_04}}
 {id=WavePlayer_12, properties {soundWaveAssetPtr=&Water_Player_Swim_Entry_3P_Slow_01}}
 {id=WavePlayer_13, properties {soundWaveAssetPtr=&Water_Player_Swim_Entry_3P_Slow_02}}
 {id=WavePlayer_14, properties {soundWaveAssetPtr=&/Water_Player_Swim_Entry_3P_Slow_03}}
 {id=WavePlayer_15, properties {soundWaveAssetPtr=&/Water_Player_Swim_Entry_3P_Slow_04}}
 {id=WavePlayer_16, properties {soundWaveAssetPtr=&Water_Player_Swim_Entry_3P_Slow_01}}
 {id=WavePlayer_17, properties {soundWaveAssetPtr=&Water_Player_Swim_Entry_3P_Slow_02}}
 {id=WavePlayer_18, properties {soundWaveAssetPtr=&Water_Player_Swim_Entry_3P_Slow_03}}
 {id=WavePlayer_19, properties {soundWaveAssetPtr=&Water_Player_Swim_Entry_3P_Slow_04}}
 {id=WavePlayer_2, properties {soundWaveAssetPtr=&Water_Player_Swim_Entry_3P_Slow_03}}
 {id=WavePlayer_20, properties {soundWaveAssetPtr=&/Water_Player_Swim_Entry_3P_Medium_01}}
 {id=WavePlayer_21, properties {soundWaveAssetPtr=&/Water_Player_Swim_Entry_3P_Medium_02}}
 {id=WavePlayer_22, properties {soundWaveAssetPtr=&/Water_Player_Swim_Entry_3P_Medium_03}}
 {id=WavePlayer_23, properties {soundWaveAssetPtr=&/Water_Player_Swim_Entry_3P_Medium_04}}
 {id=WavePlayer_24, properties {soundWaveAssetPtr=&/Water_Player_Swim_Entry_3P_Fast_01}}
 {id=WavePlayer_25, properties {soundWaveAssetPtr=&/Water_Player_Swim_Entry_3P_Fast_02}}
 {id=WavePlayer_26, properties {soundWaveAssetPtr=&Water_Player_Swim_Entry_3P_Fast_03}}
 {id=WavePlayer_27, properties {soundWaveAssetPtr=&Water_Player_Swim_Entry_3P_Fast_04}}
 {id=WavePlayer_28, properties {soundWaveAssetPtr=&Water_Player_Swim_Entry_3P_Medium_01}}
 {id=WavePlayer_29, properties {soundWaveAssetPtr=&Water_Player_Swim_Entry_3P_Medium_02}}
 {id=WavePlayer_3, properties {soundWaveAssetPtr=&Water_Player_Swim_Entry_3P_Slow_04}}
 {id=WavePlayer_30, properties {soundWaveAssetPtr=&Water_Player_Swim_Entry_3P_Medium_03}}
 {id=WavePlayer_31, properties {soundWaveAssetPtr=&Water_Player_Swim_Entry_3P_Medium_04}}
 {id=WavePlayer_32, properties {soundWaveAssetPtr=&Water_Player_Swim_Entry_3P_Fast_01}}
 {id=WavePlayer_33, properties {soundWaveAssetPtr=&Water_Player_Swim_Entry_3P_Fast_02}}
 {id=WavePlayer_34, properties {soundWaveAssetPtr=&Water_Player_Swim_Entry_3P_Fast_03}}
 {id=WavePlayer_35, properties {soundWaveAssetPtr=&Water_Player_Swim_Entry_3P_Fast_04}}
 {id=WavePlayer_4, properties {soundWaveAssetPtr=&Water_Player_Swim_Entry_3P_Medium_01}}
 {id=WavePlayer_5, properties {soundWaveAssetPtr=&Water_Player_Swim_Entry_3P_Medium_02}}
 {id=WavePlayer_6, properties {soundWaveAssetPtr=&Water_Player_Swim_Entry_3P_Medium_03}}
 {id=WavePlayer_7, properties {soundWaveAssetPtr=&Water_Player_Swim_Entry_3P_Medium_04}}
 {id=WavePlayer_8, properties {soundWaveAssetPtr=&Water_Player_Swim_Entry_3P_Fast_01}}
 {id=WavePlayer_9, properties {soundWaveAssetPtr=&Water_Player_Swim_Entry_3P_Fast_02}}
]
properties {
 attenuationSettings=&Fort_Player_Jump_3P_Athena_Water_Stereo_Att
 maxDistance=2097152
 pitchMultiplier=1.2
 soundClassObject=&Fort_Player_Foley_3P_Enemy_Athena
 volumeMultiplier=0.12
 concurrencySet [
  &/Water_Player_Swim_3P_Concurrency_Group
  &/Water_Player_Swim_Land
  &/Conc_Water_Player_Swim_3P_All
 ]
}
