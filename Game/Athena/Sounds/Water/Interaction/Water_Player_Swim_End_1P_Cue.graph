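# First-person Sound Cue for the end of player swimming in Athena, using footstep walk attenuation and the water interaction sound class.
# A Switch on the IsWaterJump int parameter is the root, but all three branches share one Random node, so every branch draws from the same four equally weighted waves.
# A swim concurrency group is assigned, alongside override settings for one instance per owner that stops the oldest.
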SoundCue /Game/Athena/Sounds/Water/Interaction/Water_Player_Swim_End_1P_Cue
root=Switch_0
nodes [
 {id=Random_0, children [WavePlayer_0,WavePlayer_2,WavePlayer_3,WavePlayer_4], properties {weights [1,1,1,1]}}
 {id=Switch_0, children [Random_0,Random_0,Random_0], properties {intParameterName=IsWaterJump}}
 {id=WavePlayer_0, properties {soundWaveAssetPtr=&/Water_Player_Swim_End_1P_01}}
 {id=WavePlayer_2, properties {soundWaveAssetPtr=&/Water_Player_Swim_End_1P_02}}
 {id=WavePlayer_3, properties {soundWaveAssetPtr=&/Water_Player_Swim_End_1P_03}}
 {id=WavePlayer_4, properties {soundWaveAssetPtr=&/Water_Player_Swim_End_1P_04}}
]
properties {
 attenuationSettings=/Game/Sounds/Attenuation/Fort_Player_Footsteps_1P_Walk_Athena_Att
 maxDistance=2097152
 soundClassObject=/Game/Sounds/FortSoundClassesAndModes/Fort_Water_Interaction
 volumeMultiplier=1.5
 concurrencyOverrides {bLimitToOwner=true, maxCount=1, resolutionRule=StopOldest}
 concurrencySet [&/Water_Player_Swim_1P_Concurrency_Group]
}
