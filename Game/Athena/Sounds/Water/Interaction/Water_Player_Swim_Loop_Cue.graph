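# Looping water sound cue for a swimming player, scaled by how fast they swim.
# A looping wave player feeds a Continuous Modulator, where the SwimVelocity parameter drives both pitch and volume, so faster movement can raise both and slow movement stays quieter.
# The cue uses player swimming attenuation and the water interaction sound class.
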SoundCue /Game/Athena/Sounds/Water/Interaction/Water_Player_Swim_Loop_Cue
root=ModulatorContinuous_0
nodes [
 {
  id=ModulatorContinuous_0
  children [WavePlayer_0]
  properties {
   pitchModulationParams {default=0, maxOutput=1.5, minOutput=1, parameterName=SwimVelocity}
   volumeModulationParams {default=0, minOutput=0.2, parameterName=SwimVelocity}
  }
 }
 {id=WavePlayer_0, properties {bLooping=true, soundWaveAssetPtr=&/Water_Player_Swim_Loop_1P_01}}
]
properties {
 attenuationSettings=/Game/Sounds/Attenuation/Fort_Player_Swimming_Athena_Att
 maxDistance=4500
 soundClassObject=/Game/Sounds/FortSoundClassesAndModes/Fort_Water_Interaction
}
