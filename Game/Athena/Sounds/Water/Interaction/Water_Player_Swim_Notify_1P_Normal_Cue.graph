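# Sound cue for the first-person normal swim notify, played in the water interaction sound class.
# A quality-level node sends its first two levels to a random pick of ten swim waves and its last two to a three-wave pick excluded from branch culling.
# A root modulator randomizes pitch and volume on each play.
# It borrows the first-person walk footstep attenuation and uses a swim concurrency group.
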SoundCue /Game/Athena/Sounds/Water/Interaction/Water_Player_Swim_Notify_1P_Normal_Cue
root=Modulator_1
nodes [
 {id=Modulator_1, children [QualityLevel_0], properties {pitchMax=2, pitchMin=1, volumeMax=1, volumeMin=0.35}}
 {id=QualityLevel_0, children [Random_2,Random_2,Random_0,Random_0]}
 {id=Random_0, children [WavePlayer_0,WavePlayer_1,WavePlayer_2], properties {bShouldExcludeFromBranchCulling=true, weights [1,1,1]}}
 {
  id=Random_2
  children [
   WavePlayer_3
   WavePlayer_4
   WavePlayer_5
   WavePlayer_6
   WavePlayer_7
   WavePlayer_8
   WavePlayer_9
   WavePlayer_10
   WavePlayer_17
   WavePlayer_18
  ]
  properties {weights [1,1,1,1,1,1,1,1,1,1]}
 }
 {id=WavePlayer_0, properties {soundWaveAssetPtr=&/Water_Player_Swim_Normal_1P_01}}
 {id=WavePlayer_1, properties {soundWaveAssetPtr=&/Water_Player_Swim_Normal_1P_02}}
 {id=WavePlayer_10, properties {soundWaveAssetPtr=&/Water_Player_Swim_Normal_1P_07}}
 {id=WavePlayer_17, properties {soundWaveAssetPtr=&/Water_Player_Swim_Normal_1P_08}}
 {id=WavePlayer_18, properties {soundWaveAssetPtr=&/Water_Player_Swim_Normal_1P_09}}
 {id=WavePlayer_2, properties {soundWaveAssetPtr=&/Water_Player_Swim_Normal_1P_03}}
 {id=WavePlayer_3, properties {soundWaveAssetPtr=&/Water_Player_Swim_Normal_1P_10}}
 {id=WavePlayer_4, properties {soundWaveAssetPtr=&Water_Player_Swim_Normal_1P_01}}
 {id=WavePlayer_5, properties {soundWaveAssetPtr=&Water_Player_Swim_Normal_1P_02}}
 {id=WavePlayer_6, properties {soundWaveAssetPtr=&Water_Player_Swim_Normal_1P_03}}
 {id=WavePlayer_7, properties {soundWaveAssetPtr=&/Water_Player_Swim_Normal_1P_04}}
 {id=WavePlayer_8, properties {soundWaveAssetPtr=&/Water_Player_Swim_Normal_1P_05}}
 {id=WavePlayer_9, properties {soundWaveAssetPtr=&/Water_Player_Swim_Normal_1P_06}}
]
properties {
 attenuationSettings=/Game/Sounds/Attenuation/Fort_Player_Footsteps_1P_Walk_Athena_Att
 maxDistance=2097152
 soundClassObject=/Game/Sounds/FortSoundClassesAndModes/Fort_Water_Interaction
 volumeMultiplier=1
 attenuationOverrides {bAttenuate=false, bAttenuateWithLPF=true, bSpatialize=false, lPFFrequencyAtMax=1000, lPFFrequencyAtMin=1000, lPFRadiusMax=0, lPFRadiusMin=0}
 concurrencyOverrides {maxCount=3, resolutionRule=StopOldest}
 concurrencySet [&/Water_Player_Swim_1P_Concurrency_Group]
}
