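# First-person sound cue for the start of a player's swim sprint boost, apparently fired from a swim animation notify.
# A Random node picks one of four equally weighted boost-start waves, and a Modulator varies its pitch and volume on each play.
# It reuses first-person footstep attenuation, routes to the water camera FX sound class, and overrides concurrency to cap instances at two, stopping the oldest.
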SoundCue /Game/Athena/Sounds/Water/Interaction/Water_Player_Swim_Notify_1P_SprintBoostStart_Cue
root=Modulator_0
nodes [
 {id=Modulator_0, children [Random_0], properties {pitchMax=1.2, pitchMin=0.9, volumeMax=1, volumeMin=0.6}}
 {id=Random_0, children [WavePlayer_0,WavePlayer_1,WavePlayer_2,WavePlayer_5], properties {weights [1,1,1,1]}}
 {id=WavePlayer_0, properties {soundWaveAssetPtr=&/Water_Player_SwimBoost_Start_1P_01}}
 {id=WavePlayer_1, properties {soundWaveAssetPtr=&/Water_Player_SwimBoost_Start_1P_02}}
 {id=WavePlayer_2, properties {soundWaveAssetPtr=&/Water_Player_SwimBoost_Start_1P_03}}
 {id=WavePlayer_5, properties {soundWaveAssetPtr=&/Water_Player_SwimBoost_Start_1P_04}}
]
properties {
 attenuationSettings=/Game/Sounds/Attenuation/Fort_Player_Footsteps_1P_Walk_Athena_Att
 bOverrideConcurrency=true
 maxDistance=2097152
 soundClassObject=/Game/Sounds/FortSoundClassesAndModes/Fort_Water_CameraFX
 volumeMultiplier=1.2
 concurrencyOverrides {maxCount=2, resolutionRule=StopOldest}
}
