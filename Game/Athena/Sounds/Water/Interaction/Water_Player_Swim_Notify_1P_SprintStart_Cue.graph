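# First-person player sound cue for starting a sprint swim, named as a swim notify in Athena's water interaction sounds.
# A Random node picks one of four equally weighted swim-start waves, which feeds a Modulator that raises pitch, with its min set above its max.
# It plays in the water interaction sound class with boosted volume and borrows first-person walk footstep attenuation.
# A swim concurrency group is assigned, alongside an override block specifying stop-oldest.
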SoundCue /Game/Athena/Sounds/Water/Interaction/Water_Player_Swim_Notify_1P_SprintStart_Cue
root=Modulator_0
nodes [
 {id=Modulator_0, children [Random_1], properties {pitchMax=1.2, pitchMin=1.3}}
 {id=Random_1, children [WavePlayer_3,WavePlayer_4,WavePlayer_5,WavePlayer_6], properties {weights [1,1,1,1]}}
 {id=WavePlayer_3, properties {soundWaveAssetPtr=&/Water_Player_Swim_Start_1P_01}}
 {id=WavePlayer_4, properties {soundWaveAssetPtr=&/Water_Player_Swim_Start_1P_02}}
 {id=WavePlayer_5, properties {soundWaveAssetPtr=&/Water_Player_Swim_Start_1P_03}}
 {id=WavePlayer_6, properties {soundWaveAssetPtr=&/Water_Player_Swim_Start_1P_04}}
]
properties {
 attenuationSettings=/Game/Sounds/Attenuation/Fort_Player_Footsteps_1P_Walk_Athena_Att
 maxDistance=2097152
 soundClassObject=/Game/Sounds/FortSoundClassesAndModes/Fort_Water_Interaction
 volumeMultiplier=1.8
 concurrencyOverrides {maxCount=2, resolutionRule=StopOldest}
 concurrencySet [&/Water_Player_Swim_1P_Concurrency_Group]
}
