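# Sound cue for first-person sprint-swimming water sounds, likely fired by animation notifies.
# A quality-level switch selects an eleven-variation random pool for the first two levels and a four-variation pool for the last two.
# A top modulator randomizes pitch and volume on each play.
# It uses the Water Interaction sound class, sprint-footstep attenuation and a shared first-person swim concurrency group.
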SoundCue /Game/Athena/Sounds/Water/Interaction/Water_Player_Swim_Notify_1P_Sprint_Cue
root=Modulator_1
nodes [
 {id=Modulator_1, children [QualityLevel_0], properties {pitchMax=1.2, pitchMin=0.9, volumeMin=0.45}}
 {id=QualityLevel_0, children [Random_1,Random_1,Random_0,Random_0]}
 {
  id=Random_0
  children [WavePlayer_1,WavePlayer_2,WavePlayer_3,WavePlayer_4]
  properties {bShouldExcludeFromBranchCulling=true, weights [1,1,1,1]}
 }
 {
  id=Random_1
  children [
   WavePlayer_0
   WavePlayer_11
   WavePlayer_12
   WavePlayer_13
   WavePlayer_14
   WavePlayer_15
   WavePlayer_16
   WavePlayer_17
   WavePlayer_18
   WavePlayer_19
   WavePlayer_20
  ]
  properties {weights [1,1,1,1,1,1,1,1,1,1,1]}
 }
 {id=WavePlayer_0, properties {soundWaveAssetPtr=&/Water_Player_Swim_Sprint_1P_10}}
 {id=WavePlayer_1, properties {soundWaveAssetPtr=&Water_Player_Swim_Sprint_1P_10}}
 {id=WavePlayer_11, properties {soundWaveAssetPtr=&/Water_Player_Swim_Sprint_1P_11}}
 {id=WavePlayer_12, properties {soundWaveAssetPtr=&/Water_Player_Swim_Sprint_1P_01}}
 {id=WavePlayer_13, properties {soundWaveAssetPtr=&/Water_Player_Swim_Sprint_1P_02}}
 {id=WavePlayer_14, properties {soundWaveAssetPtr=&/Water_Player_Swim_Sprint_1P_03}}
 {id=WavePlayer_15, properties {soundWaveAssetPtr=&/Water_Player_Swim_Sprint_1P_04}}
 {id=WavePlayer_16, properties {soundWaveAssetPtr=&/Water_Player_Swim_Sprint_1P_05}}
 {id=WavePlayer_17, properties {soundWaveAssetPtr=&/Water_Player_Swim_Sprint_1P_06}}
 {id=WavePlayer_18, properties {soundWaveAssetPtr=&/Water_Player_Swim_Sprint_1P_07}}
 {id=WavePlayer_19, properties {soundWaveAssetPtr=&/Water_Player_Swim_Sprint_1P_08}}
 {id=WavePlayer_2, properties {soundWaveAssetPtr=&Water_Player_Swim_Sprint_1P_11}}
 {id=WavePlayer_20, properties {soundWaveAssetPtr=&/Water_Player_Swim_Sprint_1P_09}}
 {id=WavePlayer_3, properties {soundWaveAssetPtr=&Water_Player_Swim_Sprint_1P_01}}
 {id=WavePlayer_4, properties {soundWaveAssetPtr=&Water_Player_Swim_Sprint_1P_02}}
]
properties {
 attenuationSettings=/Game/Sounds/Attenuation/Fort_Player_Footsteps_1P_Sprint_Athena_Att
 maxDistance=2097152
 soundClassObject=/Game/Sounds/FortSoundClassesAndModes/Fort_Water_Interaction
 volumeMultiplier=1.5
 concurrencyOverrides {maxCount=3, resolutionRule=StopOldest}
 concurrencySet [&/Water_Player_Swim_1P_Concurrency_Group]
}
