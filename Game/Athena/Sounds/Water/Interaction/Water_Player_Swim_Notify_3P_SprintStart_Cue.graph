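# Sound cue for the third-person swim sprint-start notify, heard when other players begin sprint swimming in Athena water.
# A Random node picks evenly between two swim-start wave variations.
# It routes to the water interaction sound class with boosted volume and player swimming attenuation, and two third-person swim concurrency groups cap overlapping instances.
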SoundCue /Game/Athena/Sounds/Water/Interaction/Water_Player_Swim_Notify_3P_SprintStart_Cue
root=Random_1
nodes [
 {id=Random_1, children [WavePlayer_3,WavePlayer_4], properties {weights [1,1]}}
 {id=WavePlayer_3, properties {soundWaveAssetPtr=&/Water_Player_Swim_Start_3P_01}}
 {id=WavePlayer_4, properties {soundWaveAssetPtr=&/Water_Player_Swim_Start_3P_02}}
]
properties {
 attenuationSettings=/Game/Sounds/Attenuation/Fort_Player_Swimming_Athena_Att
 maxDistance=4500
 soundClassObject=/Game/Sounds/FortSoundClassesAndModes/Fort_Water_Interaction
 volumeMultiplier=1.5
 concurrencySet [
  &/Water_Player_Swim_3P_Concurrency_Group
  &/Conc_Water_Player_Swim_3P_All
 ]
}
