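# Sound cue for the first-person player swim start sound in Athena water interaction.
# A switch on the IsWaterJump parameter routes all three of its inputs to the same equally weighted random pick of four waves, so the parameter does not change the selection.
# It uses first-person walk footstep attenuation, the water interaction sound class, a boosted volume and a shared swim concurrency group.
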SoundCue /Game/Athena/Sounds/Water/Interaction/Water_Player_Swim_Start_1P_Cue
root=Switch_0
nodes [
 {id=Random_1, children [WavePlayer_0,WavePlayer_1,WavePlayer_2,WavePlayer_5], properties {weights [1,1,1,1]}}
 {id=Switch_0, children [Random_1,Random_1,Random_1], properties {intParameterName=IsWaterJump}}
 {id=WavePlayer_0, properties {soundWaveAssetPtr=&/Water_Player_Swim_Start_1P_01}}
 {id=WavePlayer_1, properties {soundWaveAssetPtr=&/Water_Player_Swim_Start_1P_02}}
 {id=WavePlayer_2, properties {soundWaveAssetPtr=&/Water_Player_Swim_Start_1P_03}}
 {id=WavePlayer_5, properties {soundWaveAssetPtr=&/Water_Player_Swim_Start_1P_04}}
]
properties {
 attenuationSettings=/Game/Sounds/Attenuation/Fort_Player_Footsteps_1P_Walk_Athena_Att
 maxDistance=2097152
 soundClassObject=/Game/Sounds/FortSoundClassesAndModes/Fort_Water_Interaction
 volumeMultiplier=1.5
 concurrencyOverrides {bLimitToOwner=true, maxCount=2, resolutionRule=StopOldest}
 concurrencySet [&/Water_Player_Swim_1P_Concurrency_Group]
}
