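# Sound cue for the third-person sound of a player starting to swim in Athena.
# An IsEnemy branch plays at full mixer volume for enemies or when the parameter is unset, and much quieter for non-enemies.
# Each path randomly picks between two equally weighted swim-start waves, with both random nodes excluded from branch culling.
# Swimming attenuation, the water-interaction sound class and two swim concurrency groups govern playback.
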SoundCue /Game/Athena/Sounds/Water/Interaction/Water_Player_Swim_Start_3P_Cue
root=Branch_1
nodes [
 {id=Branch_1, children [Mixer_1,Mixer_0,Mixer_1], properties {boolParameterName=IsEnemy}}
 {id=Mixer_0, children [Random_0], properties {inputVolume [0.3]}}
 {id=Mixer_1, children [Random_1], properties {inputVolume [1]}}
 {id=Random_0, children [WavePlayer_0,WavePlayer_1], properties {bShouldExcludeFromBranchCulling=true, weights [1,1]}}
 {id=Random_1, children [WavePlayer_2,WavePlayer_3], properties {bShouldExcludeFromBranchCulling=true, weights [1,1]}}
 {id=WavePlayer_0, properties {soundWaveAssetPtr=&/Water_Player_Swim_Start_3P_01}}
 {id=WavePlayer_1, properties {soundWaveAssetPtr=&/Water_Player_Swim_Start_3P_02}}
 {id=WavePlayer_2, properties {soundWaveAssetPtr=&Water_Player_Swim_Start_3P_01}}
 {id=WavePlayer_3, properties {soundWaveAssetPtr=&Water_Player_Swim_Start_3P_02}}
]
properties {
 attenuationSettings=/Game/Sounds/Attenuation/Fort_Player_Swimming_Athena_Att
 maxDistance=4500
 soundClassObject=/Game/Sounds/FortSoundClassesAndModes/Fort_Water_Interaction
 volumeMultiplier=1.75
 concurrencyOverrides {bLimitToOwner=true, maxCount=2, resolutionRule=StopOldest}
 concurrencySet [
  &/Water_Player_Swim_3P_Concurrency_Group
  &/Conc_Water_Player_Swim_3P_All
 ]
}
