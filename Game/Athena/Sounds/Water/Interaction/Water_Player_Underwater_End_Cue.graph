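# Sound cue named for the end of a player being underwater, in the Athena water interaction sounds.
# It has no nodes and no root, so it produces no audio as it stands.
# It still carries playback settings: the water interaction sound class, a third-person walk footstep attenuation with its max distance, and a reduced volume multiplier.
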SoundCue /Game/Athena/Sounds/Water/Interaction/Water_Player_Underwater_End_Cue
root=null
nodes []
properties {
 attenuationSettings=/Game/Sounds/Attenuation/Fort_Player_Footsteps_3P_Walk_Athena_Att
 maxDistance=3584
 soundClassObject=/Game/Sounds/FortSoundClassesAndModes/Fort_Water_Interaction
 volumeMultiplier=0.6
}
