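# Sound cue for a looping underwater sound tied to player water interaction in Athena.
# A single looping wave player plays the Amb_Water_Underwater_Loop_01 ambient wave straight to the output, slightly reduced in volume.
# It routes through the water camera FX sound class and reuses a third-person player walk footstep attenuation for spatial falloff.
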SoundCue /Game/Athena/Sounds/Water/Interaction/Water_Player_Underwater_Loop_Cue
root=WavePlayer_0
nodes [{id=WavePlayer_0, properties {bLooping=true, soundWaveAssetPtr=/Game/Athena/Sounds/Water/Ambient/Amb_Water_Underwater_Loop_01}}]
properties {
 attenuationSettings=/Game/Sounds/Attenuation/Fort_Player_Footsteps_3P_Walk_Athena_Att
 maxDistance=3584
 soundClassObject=/Game/Sounds/FortSoundClassesAndModes/Fort_Water_CameraFX
 volumeMultiplier=0.8
}
