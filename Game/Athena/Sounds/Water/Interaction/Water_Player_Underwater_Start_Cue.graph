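# Sound cue for a player starting to go underwater in Athena, judging by its name.
# A Random node picks among ten equally weighted first-person normal swim wave variations, reusing swim stroke sounds rather than dedicated submerge samples.
# It plays at reduced volume through the water interaction sound class, spatialized with the third-person walking footstep attenuation asset.
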SoundCue /Game/Athena/Sounds/Water/Interaction/Water_Player_Underwater_Start_Cue
root=Random_0
nodes [
 {
  id=Random_0
  children [WavePlayer_0, WavePlayer_2, WavePlayer_3, WavePlayer_4, WavePlayer_5, WavePlayer_6, WavePlayer_7, WavePlayer_8, WavePlayer_9, WavePlayer_10]
  properties {weights [1,1,1,1,1,1,1,1,1,1]}
 }
 {id=WavePlayer_0, properties {soundWaveAssetPtr=&/Water_Player_Swim_Normal_1P_03}}
 {id=WavePlayer_10, properties {soundWaveAssetPtr=&/Water_Player_Swim_Normal_1P_09}}
 {id=WavePlayer_2, properties {soundWaveAssetPtr=&/Water_Player_Swim_Normal_1P_07}}
 {id=WavePlayer_3, properties {soundWaveAssetPtr=&/Water_Player_Swim_Normal_1P_10}}
 {id=WavePlayer_4, properties {soundWaveAssetPtr=&/Water_Player_Swim_Normal_1P_01}}
 {id=WavePlayer_5, properties {soundWaveAssetPtr=&/Water_Player_Swim_Normal_1P_02}}
 {id=WavePlayer_6, properties {soundWaveAssetPtr=&/Water_Player_Swim_Normal_1P_04}}
 {id=WavePlayer_7, properties {soundWaveAssetPtr=&/Water_Player_Swim_Normal_1P_05}}
 {id=WavePlayer_8, properties {soundWaveAssetPtr=&/Water_Player_Swim_Normal_1P_06}}
 {id=WavePlayer_9, properties {soundWaveAssetPtr=&/Water_Player_Swim_Normal_1P_08}}
]
properties {
 attenuationSettings=/Game/Sounds/Attenuation/Fort_Player_Footsteps_3P_Walk_Athena_Att
 maxDistance=3584
 soundClassObject=/Game/Sounds/FortSoundClassesAndModes/Fort_Water_Interaction
 volumeMultiplier=0.6
}
